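# M_Hivemind is a swarm bug material that animates instanced and Niagara mesh insects procedurally through world position offset.
# Vertex colors and position build body-part masks, which select pivots, axes and rotation amounts for legs, torso, mouth, wings and antlers.
# Time-driven samples of a curve atlas drive those rotations.
# Animation states (idle, walk, attack, falling, death and others) come from custom primitive data or dynamic parameters and blend, with static-switch preview modes.
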Material /ProtoSwarm/NPCs/Swarm/TidalCrane_Swarm_WPOTest/Materials/M_Hivemind
graph {
 properties {
  bAutomaticallySetUsageInEditor=false
  bTangentSpaceNormal=false
  bUseMaterialAttributes=true
  bUsedWithInstancedStaticMeshes=true
  bUsedWithNiagaraMeshParticles=true
  previewMesh=/ProtoSwarm/NPCs/Swarm/TidalCrane_Swarm_WPOTest/Meshes/SM_BugProto_TA2
  usedShadingModels=Default Lit
 }
 nodes [
  Constant3Vector()
  Constant() {r=1}
  Constant()
  Constant() {r=7}
  n1 = Time()
  n2 = Multiply(A=Sine(None=Divide(A=@n1) {constB=10})) {constB=2}
  n3 = Add(A=Divide(A=Time(), B=Saturate(None=@n2)), B=NamedRerouteUsage())
  Cosine(None=@n3)
  n4 = Time()
  n5 = ScalarParameter() {defaultValue=0.5, group=WPO, parameterName=Attack_Duration, primitiveDataIndex=1, sortPriority=1}
  Divide(A=@n4, B=@n5)
  n6 = Constant() {r=1.09}
  Floor(None=@n6)
  n7 = Constant3Vector() {constant=<rgba 22 25 60 0>}
  n8 = NamedRerouteUsage()
  n9 = Multiply(A=Constant3Vector() {constant=<rgba 40 0 60 0>}, B=NamedRerouteUsage())
  n10 = Multiply(A=Constant3Vector() {constant=<rgba 30 -40 60 0>}, B=NamedRerouteUsage())
  n11 = Multiply(A=Add(A=Multiply(A=@n7, B=@n8), B=Add(A=@n9, B=@n10)), B=NamedRerouteUsage())
  n12 = Step(X=Subtract(A=LocalPosition().XYZ, B=NamedRerouteDeclaration(None=@n11) {name=Legs_Pivot Point_C, nodeColor=<rgba 1 0.7734 0 1>}))
  LinearInterpolate(A=Multiply(A=@n12) {constB=0.05}, B=@n12, Alpha=NamedRerouteUsage())
  n13 = Reroute()
  n14 = OneMinus(None=@n13)
  n15 = Add(A=Multiply(A=@n14, B=NamedRerouteUsage()), B=Multiply(A=@n13, B=NamedRerouteUsage()))
  MaterialFunctionCall(Float4=@n15) {materialFunction=&BreakOutFloat4Components}
  n16 = Time()
  n17 = Constant3Vector() {constant=<rgba 2 0.7 0.3 0>}
  MaterialFunctionCall(Color=Add(A=Divide(A=@n16, B=@n17), B=ScalarParameter() {bUseCustomPrimitiveData=true, parameterName=RandomTime, primitiveDataIndex=4})) {materialFunction=&SplitComponents}
  n18 = Add(A=Multiply(A=@n14, B=NamedRerouteUsage()), B=Multiply(A=@n13, B=NamedRerouteUsage()))
  MaterialFunctionCall(Float4=@n18) {materialFunction=&BreakOutFloat4Components}
  n19 = Reroute()
  n20 = Reroute()
  n21 = Subtract(A=@n19, B=@n20)
  MaterialFunctionCall!(Vector3=@n21, MaximumNumberOfDigits, UVs, DebugTextLocation RG_UpperRight BA_LowerLeft, Component Spacing) {materialFunction=/Engine/Functions/Engine_MaterialFunctions02/Utility/DebugFloat3Values}
  MaterialFunctionCall!(Vector2=AppendVector(A=@n19, B=@n20), MaximumNumberOfDigits, UVs, DebugTextLocation RG_UpperRight BA_LowerLeft, Component Spacing) {materialFunction=/Engine/Functions/Engine_MaterialFunctions02/Utility/DebugFloat2Values}
  n22 = Constant()
  n23 = Constant() {r=7}
  Multiply(
   A=NamedRerouteUsage()
   B=Multiply(
    A=MaterialFunctionCall(R=@n22, G=@n22, B=@n23) {materialFunction=&AppendMany}.RGB
    B=NamedRerouteUsage()
   )
  )
  NamedRerouteDeclaration(None=DynamicParameter() {parameterIndex=2, defaultValue=<rgba 0 0 0 0>, paramNames [LegsAnimTime, "", "", ""]}.LegsAnimTime) {name=VFX_LegsAnimTime, nodeColor=<rgba 0 0.25 1 1>}
  n24 = DynamicParameter() {defaultValue=<rgba 1 0 0 0>, paramNames [Flying,Idle,Walk,Sprint]}
  NamedRerouteDeclaration(None=@n24.Flying) {name=VFX_Fly, nodeColor=<rgba 0 0.25 1 1>}
  NamedRerouteDeclaration(None=@n24.Idle) {name=VFX_Idle, nodeColor=<rgba 0 0.25 1 1>}
  NamedRerouteDeclaration(None=@n24.Walk) {name=VFX_Walk, nodeColor=<rgba 0 0.25 1 1>}
  n25 = DynamicParameter() {parameterIndex=1, defaultValue=<rgba 0 0 0 0>, paramNames [LegsAnimTime, "", "", ""]}
  NamedRerouteDeclaration(None=@n25.LegsAnimTime) {name=VFX_Attack, nodeColor=<rgba 0 0.25 1 1>}
  NamedRerouteDeclaration(None=@n25.G) {name=VFX_Falling, nodeColor=<rgba 0 0.25 1 1>}
  NamedRerouteDeclaration(None=@n25.A) {name=VFX_Death, nodeColor=<rgba 0 0.25 1 1>}
  NamedRerouteDeclaration(None=@n25.B) {name=VFX_Flower, nodeColor=<rgba 0 0.25 1 1>}
  n26 = ScalarParameter() {
   defaultValue=1
   desc="0 - Idle\r\n1 - Walk\r\n2 - MeleeAttack\r\n3- Falling\r\n4 - Death\r\n5 - Flower\r\n6 - Jump"
   group=WPO
   parameterName="___Preview Anim (Index)"
   sortPriority=-10
  }
  n27 = ScalarParameter() {
   defaultValue=5
   desc="0 - Idle\r\n1 - Walk\r\n2 - MeleeAttack\r\n3- Falling\r\n4 - Death\r\n5 - Flower\r\n6 - Jump"
   group=WPO
   parameterName=___Preview AnimA
   sortPriority=-5
  }
  n28 = Floor(None=NamedRerouteUsage())
  n29 = NamedRerouteDeclaration(None=StaticBoolParameter() {group=WPO, parameterName="• AnimBlendPreview", sortPriority=-5}) {name=DebugAnim, nodeColor=<rgba 1 0.0469 0 1>}
  n30 = StaticBoolParameter() {group=WPO, parameterName="• SingleAnimPreview", sortPriority=-10}
  NamedRerouteDeclaration(None=StaticSwitch(True=@n26, False=StaticSwitch(True=@n27, False=@n28, Value=@n29), Value=@n30)) {name=StateA_Value, nodeColor=<rgba 1 0.5625 0 1>}
  n31 = ScalarParameter() {
   defaultValue=7
   desc="0 - Idle\r\n1 - Walk\r\n2 - MeleeAttack\r\n3- Falling\r\n4 - Death\r\n5 - Flower\r\n6 - Jump"
   group=WPO
   parameterName=___Preview AnimB
   sortPriority=-5
  }
  n32 = Floor(None=NamedRerouteUsage())
  NamedRerouteDeclaration(None=StaticSwitch(True=@n26, False=StaticSwitch(True=@n31, False=@n32, Value=@n29), Value=@n30)) {name=StateB_Value, nodeColor=<rgba 0.2422 0 1 1>}
  n33 = NamedRerouteUsage()
  n34 = Constant4Vector() {desc="0 - Idle\r\n1 - Walk\r\n2 - MeleeAttack\r\n3- Falling\r\n4 - Death\r\n5 - Flower\r\n6 - Jump", constant=<rgba 0 1 2 3>}
  NamedRerouteDeclaration(None=Multiply(A=Step(Y=@n34.RGBA, X=@n33), B=Step(Y=@n33, X=@n34.RGBA))) {name=IndexA_Validation 1, nodeColor=<rgba 0.0781 1 0 1>}
  n35 = NamedRerouteUsage()
  n36 = Constant4Vector() {desc="0 - Idle\r\n1 - Walk\r\n2 - MeleeAttack\r\n3- Falling\r\n4 - Death\r\n5 - Flower\r\n6 - Jump", constant=<rgba 4 5 6 7>}
  NamedRerouteDeclaration(None=Multiply(A=Step(Y=@n36.RGBA, X=@n35), B=Step(Y=@n35, X=@n36.RGBA))) {name=IndexA_Validation 2, nodeColor=<rgba 0.0781 1 0 1>}
  n37 = NamedRerouteUsage()
  NamedRerouteDeclaration(None=Multiply(A=Step(Y=@n34.RGBA, X=@n37), B=Step(Y=@n37, X=@n34.RGBA))) {name=IndexB_Validation 1, nodeColor=<rgba 0.4429 0 1 1>}
  n38 = NamedRerouteUsage()
  NamedRerouteDeclaration(None=Multiply(A=Step(Y=@n36.RGBA, X=@n38), B=Step(Y=@n38, X=@n36.RGBA))) {name=IndexB_Validation 2, nodeColor=<rgba 0.4429 0 1 1>}
  NamedRerouteDeclaration(None=@n24.Sprint) {name=VFX_Sprint, nodeColor=<rgba 0 0.25 1 1>}
  n39 = VectorParameter() {bUseCustomPrimitiveData=true, parameterName=AnimationBlend, primitiveDataIndex=10}
  n40 = NamedRerouteDeclaration(None=@n39.B) {name=PD_StateA, nodeColor=<rgba 0 1 0.1094 1>}
  n41 = NamedRerouteDeclaration(None=@n39.A) {name=PD_StateB, nodeColor=<rgba 0.4297 0 1 1>}
  n42 = NamedRerouteUsage()
  NamedRerouteDeclaration(None=Fmod(A=LinearInterpolate(A=@n40, B=@n41, Alpha=@n42), B=Constant() {r=1})) {name=PD_BlendDur, nodeColor=<rgba 0 1 0.6953 1>}
  n43 = NamedRerouteUsage()
  n44 = Multiply(A=Constant() {r=-0.18}, B=NamedRerouteUsage())
  n45 = AppendVector(A=Add(A=Add(A=@n43, B=@n44)) {constB=0.25}, B=NamedRerouteUsage())
  NamedRerouteDeclaration(None=TextureSample(UVs=Frac(None=@n45), Tex=NamedRerouteUsage()) {samplerType=SAMPLERTYPE_LinearColor, texture=/ProtoSwarm/NPCs/Swarm/TidalCrane_Swarm_WPOTest/Textures/CA_SwarmerWPOCurves}.R) {name=Curve_Torso_Walk_Rot1_Delay, desc=legs torso, nodeColor=<rgba 1 0 0.6562 1>}
  n46 = Multiply(A=Constant3Vector() {desc=50, constant=<rgba 142 134 0 0>}, B=NamedRerouteUsage())
  n47 = Add(
   A=Multiply(A=Constant3Vector() {constant=<rgba 134 30 0 0>}, B=NamedRerouteUsage())
   B=Multiply(A=Constant3Vector() {constant=<rgba 73 -175 0 0>}, B=NamedRerouteUsage())
  )
  NamedRerouteDeclaration(None=Multiply(A=Add(A=@n46, B=@n47), B=NamedRerouteUsage())) {name=Legs_Pivot Point_A, nodeColor=<rgba 1 0.7734 0 1>}
  n48 = NamedRerouteUsage()
  n49 = AppendVector(A=Add(A=@n48, B=@n44) {constB=0.25}, B=NamedRerouteUsage())
  NamedRerouteDeclaration(None=TextureSample(UVs=Frac(None=@n49), Tex=NamedRerouteUsage()) {samplerType=SAMPLERTYPE_LinearColor, texture=&CA_SwarmerWPOCurves}.R) {name=Curve_Torso_Loop_Rot1_Delay, nodeColor=<rgba 1 0 0.6562 1>}
  n50 = LocalPosition() {localOrigin=InstancePreSkinning}
  n51 = ShaderStageSwitch(PixelShader=VertexInterpolator(VS=@n50.XYZ).PS, VertexShader=@n50.XYZ)
  n52 = ComponentMask(None=@n51) {r=true}
  NamedRerouteDeclaration(None=Step(X=@n52)) {name=M_Left/Right, nodeColor=<rgba 0 0.0625 1 1>}
  n53 = TextureObject() {samplerType=SAMPLERTYPE_LinearColor, texture=&CA_SwarmerWPOCurves}
  n54 = ComponentMask(None=TextureProperty(Tex=NamedRerouteDeclaration(None=@n53) {name=CA_Tex, nodeColor=<rgba 0.4531 0 1 1>})) {g=true}
  NamedRerouteDeclaration(None=Divide(A=Constant() {desc="Curve Index (0.5,1.5,2.5...)", r=3.5}, B=@n54)) {name=CA_ROW_03_Torso, nodeColor=<rgba 1 0 0.5461 1>}
  NamedRerouteDeclaration(None=Divide(A=Constant() {desc="Curve Index (0.5,1.5,2.5...)", r=0.5}, B=@n54)) {name=CA_ROW_1, nodeColor=<rgba 0 0.3672 1 1>}
  n55 = NamedRerouteUsage()
  n56 = Divide(A=Abs(None=Subtract(A=@n4, B=@n55)), B=@n5)
  NamedRerouteDeclaration(None=Multiply(A=Saturate(None=@n56), B=Saturate(None=@n55))) {name=Time_No Loop, nodeColor=<rgba 0.6406 0 1 1>}
  n57 = NamedRerouteUsage()
  n58 = TextureSample(UVs=Frac(None=AppendVector(A=@n43, B=@n57)), Tex=NamedRerouteUsage()) {samplerType=SAMPLERTYPE_LinearColor, texture=&CA_SwarmerWPOCurves}
  NamedRerouteDeclaration(None=@n58.G) {name=Curve_Torso_Walk_Rot2, desc=legs torso, nodeColor=<rgba 1 0 0.6562 1>}
  NamedRerouteDeclaration(None=@n58.B) {name=Curve_Torso_Walk_Rot2_Delay, desc=legs torso, nodeColor=<rgba 1 0 0.6562 1>}
  n59 = NamedRerouteUsage()
  n60 = TextureSample(UVs=Frac(None=AppendVector(A=@n48, B=@n59)), Tex=NamedRerouteUsage()) {samplerType=SAMPLERTYPE_LinearColor, texture=&CA_SwarmerWPOCurves}
  NamedRerouteDeclaration(None=@n60.R) {name=Curve_Torso_Loop_Rot1, nodeColor=<rgba 1 0 0.6562 1>}
  NamedRerouteDeclaration(None=@n60.G) {name=Curve_Torso_Loop_Rot1_2x, nodeColor=<rgba 1 0 0.6562 1>}
  NamedRerouteDeclaration(None=Divide(A=Constant() {desc="Curve Index (0.5,1.5,2.5...)", r=4.5}, B=@n54)) {name=CA_ROW_04_Mouth, nodeColor=<rgba 0.3197 0.614 1 1>}
  n61 = NamedRerouteUsage()
  n62 = Time()
  n63 = Constant() {r=5}
  n64 = Sine(None=Divide(A=@n62, B=@n63) {constB=5})
  n65 = NamedRerouteUsage()
  n66 = Abs(None=Subtract(A=Time(), B=@n65))
  n67 = NamedRerouteUsage()
  n68 = Multiply(A=Saturate(None=Divide(A=@n66, B=@n67)), B=Saturate(None=@n65))
  n69 = Multiply(A=@n68, B=@n68)
  n70 = Multiply(A=@n69, B=@n69)
  n71 = LinearInterpolate(A=OneMinus(None=@n70), B=@n70, Alpha=NamedRerouteUsage()) {constAlpha=0.4}
  n72 = StaticSwitch(True=Saturate(None=Multiply(A=@n64) {constB=2}), False=@n71, Value=NamedRerouteUsage())
  n73 = OneMinus(None=@n72)
  n74 = NamedRerouteUsage()
  n75 = NamedRerouteUsage()
  n76 = MaterialFunctionCall(Float4=Add(A=Multiply(A=@n73, B=@n74), B=Multiply(A=@n72, B=@n75))) {materialFunction=&BreakOutFloat4Components}
  n77 = NamedRerouteUsage()
  NamedRerouteDeclaration(None=NamedRerouteDeclaration(None=StaticSwitch(True=@n61, False=@n76.R, Value=@n77)) {name=Anim_4_Attack, nodeColor=<rgba 1 0.0722 0.2096 1>}) {name=Anim_NotLooped, nodeColor=<rgba 0.4531 0 1 1>}
  NamedRerouteDeclaration(None=@n60.B) {name=Curve_Torso_Loop_Rot1_Delayflat, nodeColor=<rgba 1 0 0.6562 1>}
  n78 = NamedRerouteUsage()
  n79 = NamedRerouteUsage()
  n80 = Add(A=@n78, B=@n79)
  n81 = Frac(None=AppendVector(A=Add(A=@n80) {constB=0.25}, B=NamedRerouteUsage()))
  n82 = NamedRerouteUsage()
  n83 = NamedRerouteUsage()
  n84 = Add(A=@n83, B=@n79)
  n85 = Frac(None=AppendVector(A=Add(A=@n84) {constB=0.25}, B=NamedRerouteUsage()))
  n86 = NamedRerouteUsage()
  n87 = NamedRerouteUsage()
  NamedRerouteDeclaration(None=LinearInterpolate(
    A=TextureSample(UVs=@n81, Tex=@n82) {samplerType=SAMPLERTYPE_LinearColor, texture=&CA_SwarmerWPOCurves}.R
    B=TextureSample(UVs=@n85, Tex=@n86) {samplerType=SAMPLERTYPE_LinearColor, texture=&CA_SwarmerWPOCurves}.R
    Alpha=@n87
   )) {name=Curve_Legs_Walk_T_Y, nodeColor=<rgba 1 0 0.0807 1>}
  n88 = Divide(A=@n62, B=@n63) {constB=5}
  NamedRerouteDeclaration(None=Step(Y=Frac(None=@n88)) {constX=0.5}) {name=Preview_BlendAorB, nodeColor=<rgba 0 0.7422 1 1>}
  NamedRerouteDeclaration(None=@n71) {name=Debug_AnimBlending, nodeColor=<rgba 1 0.1875 0 1>}
  n89 = Frac(None=AppendVector(A=Add(A=@n78, B=@n79), B=NamedRerouteUsage()))
  n90 = NamedRerouteUsage()
  n91 = Frac(None=AppendVector(A=Add(A=@n83, B=@n79), B=NamedRerouteUsage()))
  n92 = NamedRerouteUsage()
  NamedRerouteDeclaration(None=LinearInterpolate(
    A=TextureSample(UVs=@n89, Tex=@n90) {samplerType=SAMPLERTYPE_LinearColor, texture=&CA_SwarmerWPOCurves}.R
    B=TextureSample(UVs=@n91, Tex=@n92) {samplerType=SAMPLERTYPE_LinearColor, texture=&CA_SwarmerWPOCurves}.R
    Alpha=@n87
   )) {name=Curve_Legs_Walk_T_Z, nodeColor=<rgba 1 0 0.0807 1>}
  n93 = Add(A=NamedRerouteUsage()) {constB=0.25}
  n94 = NamedRerouteUsage()
  n95 = TextureSample(UVs=Frac(None=AppendVector(A=@n93, B=@n94)), Tex=NamedRerouteUsage()) {samplerType=SAMPLERTYPE_LinearColor, texture=&CA_SwarmerWPOCurves}
  NamedRerouteDeclaration(None=@n95.R) {name=Curve_Legs_Loop_Rot1, nodeColor=<rgba 1 0 0.0807 1>}
  NamedRerouteDeclaration(None=SphereMask(A=@n51, B=Constant3Vector() {constant=<rgba 0 153 100 0>}) {attenuationRadius=130}) {name=M_HeadSphere, nodeColor=<rgba 1 0 0.7031 1>}
  n96 = AppendVector(A=NamedRerouteUsage(), B=NamedRerouteUsage())
  NamedRerouteDeclaration(None=TextureSample(UVs=Frac(None=@n96), Tex=NamedRerouteUsage()) {samplerType=SAMPLERTYPE_LinearColor, texture=&CA_SwarmerWPOCurves}.B) {name=Curve_Legs_Rot1_MeleeAttack, nodeColor=<rgba 1 0 0.0807 1>}
  n97 = NamedRerouteUsage()
  n98 = Divide(A=Time(), B=Constant() {r=0.3})
  n99 = NamedRerouteUsage()
  n100 = StaticSwitch(
   True=ParticleRandom()
   False=ScalarParameter() {bUseCustomPrimitiveData=true, parameterName=RandomTime, primitiveDataIndex=4}
   Value=NamedRerouteUsage()
  )
  NamedRerouteDeclaration(None=Add(A=StaticSwitch(True=@n97, False=@n98, Value=@n99), B=@n100)) {name=Time_Primary_Sprint, nodeColor=<rgba 0.7578 1 0 1>}
  NamedRerouteDeclaration(None=@n95.A) {name=Curve_Legs_Loop_Rot1_Fast, nodeColor=<rgba 1 0 0.0807 1>}
  n101 = Add(A=Multiply(A=NamedRerouteUsage()) {constB=0.21}, B=Multiply(A=NamedRerouteUsage()) {constB=0.12})
  n102 = NamedRerouteUsage()
  n103 = LinearInterpolate(A=OneMinus(None=@n102), B=@n102, Alpha=NamedRerouteUsage())
  n104 = NamedRerouteUsage()
  NamedRerouteDeclaration(None=LinearInterpolate(A=@n101, B=Add(A=@n101) {constB=0.5}, Alpha=Multiply(A=@n103, B=@n104))) {name=TimeOffset_legs, nodeColor=<rgba 0.2656 0 1 1>}
  n105 = VertexColor()
  n106 = Multiply(A=@n105.B, B=@n105.R)
  n107 = NamedRerouteUsage()
  n108 = NamedRerouteDeclaration(None=Step(X=Subtract(A=@n106, B=@n107)) {constY=0.1}) {name=M_Wings, nodeColor=<rgba 0 1 0.9766 1>}
  n109 = NamedRerouteUsage()
  n110 = Subtract(A=Subtract(A=@n105.R, B=@n108), B=NamedRerouteDeclaration(None=Multiply(A=@n106, B=@n109)) {name=M_Mouth, nodeColor=<rgba 0.3828 1 0 1>})
  n111 = NamedRerouteUsage()
  NamedRerouteDeclaration(None=Saturate(None=Subtract(A=@n110, B=@n111))) {name=M_Abdomen, nodeColor=<rgba 1 0.6562 0 1>}
  NamedRerouteDeclaration(None=Divide(A=Constant() {desc="Curve Index (0.5,1.5,2.5...)", r=1.5}, B=@n54)) {name=CA_ROW_02_Legs, nodeColor=<rgba 0.1425 0 1 1>}
  n112 = Max(A=@n105.R, B=@n105.G)
  NamedRerouteDeclaration(None=Step(Y=Max(A=@n112, B=@n105.B)) {constX=0.1}) {name=M_Torax, nodeColor=<rgba 0.3359 0 1 1>}
  n113 = Multiply(A=Divide(A=NamedRerouteUsage(), B=Constant() {r=-75}), B=NamedRerouteUsage())
  n114 = NamedRerouteUsage()
  n115 = Constant() {r=-15}
  n116 = Constant() {r=-0.05}
  n117 = Constant() {r=-0.02}
  n118 = NamedRerouteUsage()
  n119 = Multiply(A=Add(A=Divide(A=@n114, B=@n115), B=LinearInterpolate(A=@n116, B=@n117, Alpha=@n118)) {constB=-0.05}, B=NamedRerouteUsage())
  n120 = NamedRerouteUsage()
  n121 = Multiply(A=Add(A=Divide(A=@n120) {constB=50}, B=Constant() {r=0.1}), B=NamedRerouteUsage()) {constA=0.2, constB=0.1}
  n122 = NamedRerouteUsage()
  n123 = Multiply(A=NamedRerouteUsage()) {constB=0.01}
  n124 = LinearInterpolate(Alpha=NamedRerouteUsage()) {constA=-0.1, constB=-0.23}
  n125 = Add(
   A=Add(A=Add(A=@n113, B=@n119), B=Multiply(A=@n121, B=@n122) {constA=0.1})
   B=Multiply(A=Add(A=@n123, B=@n124) {constB=-0.1}, B=NamedRerouteUsage()) {constA=0.1}
  )
  n126 = LinearInterpolate(A=@n125, B=LinearInterpolate(A=@n125, Alpha=NamedRerouteUsage()) {constB=0.05}, Alpha=NamedRerouteUsage())
  n127 = NamedRerouteUsage()
  n128 = LinearInterpolate(
   A=Multiply(A=Multiply(A=@n127) {constB=0.1}, B=NamedRerouteUsage())
   B=Multiply(A=NamedRerouteUsage()) {constB=0.21}
   Alpha=NamedRerouteUsage()
  )
  n129 = NamedRerouteUsage()
  NamedRerouteDeclaration(None=Multiply(A=LinearInterpolate(A=@n126, B=@n128, Alpha=@n129), B=NamedRerouteUsage())) {name=Rot_Legs, nodeColor=<rgba 0 1 0.2969 1>}
  n130 = Step(X=@n105.G) {constY=0.7}
  n131 = NamedRerouteDeclaration(None=Step(X=@n105.G) {constY=0.9}) {name=M_Legs_Front, nodeColor=<rgba 0 0.2969 1 1>}
  n132 = Subtract(A=OneMinus(None=@n130), B=@n131)
  n133 = NamedRerouteDeclaration(None=Step(X=@n105.G) {constY=0.1}) {name=M_Legs, nodeColor=<rgba 0.6641 0 1 1>}
  NamedRerouteDeclaration(None=Saturate(None=Multiply(A=@n132, B=@n133))) {name=M_Legs_Back, nodeColor=<rgba 0 0.2969 1 1>}
  NamedRerouteDeclaration(None=Saturate(None=Subtract(A=@n130, B=@n131))) {name=M_Legs_Middle, nodeColor=<rgba 0 0.2969 1 1>}
  n134 = NamedRerouteUsage()
  n135 = Multiply(A=@n105.R, B=@n134)
  NamedRerouteDeclaration(None=@n135) {name=M_KneesGradient, nodeColor=<rgba 0.7578 0 1 1>}
  n136 = NamedRerouteDeclaration(None=@n105.A) {name=M_AGradients, nodeColor=<rgba 1 0.0234 0 1>}
  n137 = Multiply(A=@n136, B=@n134)
  NamedRerouteDeclaration(None=OneMinus(None=NamedRerouteDeclaration(None=@n137) {name=M_LegsGradient, nodeColor=<rgba 1 0.2521 0.4356 1>})) {name=M_TorsoGradient, nodeColor=<rgba 0.8281 0 1 1>}
  n138 = Multiply(A=NamedRerouteUsage(), B=NamedRerouteUsage())
  NamedRerouteDeclaration(None=Multiply(A=Multiply(A=@n138) {constB=0.05}, B=NamedRerouteUsage())) {name=Rot_Mouth, nodeColor=<rgba 1 0 0.0234 1>}
  n139 = NamedRerouteUsage()
  n140 = NamedRerouteUsage()
  n141 = Add(
   A=NamedRerouteDeclaration(None=Add(A=@n139, B=@n140) {constB=0}) {name=WPO_BeforeBodyRotation, nodeColor=<rgba 0.125 0 1 1>}
   B=NamedRerouteUsage()
  )
  n142 = LinearInterpolate(A=@n141, B=Add(A=@n141, B=Constant3Vector() {constant=<rgba 0 0 -90 0>}), Alpha=NamedRerouteUsage()) {constB=0}
  n143 = NamedRerouteUsage()
  NamedRerouteDeclaration(None=Transform(None=Add(A=@n142, B=@n143)) {transformSourceType=TRANSFORMSOURCE_Local}) {name=WPO, nodeColor=<rgba 0 0.9297 1 1>}
  n144 = Constant3Vector() {constant=<rgba 0 1 0 0>}
  n145 = Max(A=NamedRerouteUsage(), B=NamedRerouteUsage())
  n146 = Constant3Vector() {constant=<rgba 1 0 0 0>}
  n147 = NamedRerouteUsage()
  n148 = LinearInterpolate(A=Constant3Vector() {constant=<rgba 0 0 1 0>}, B=Constant3Vector() {constant=<rgba 1 1.5 0 0>}, Alpha=NamedRerouteUsage())
  n149 = NamedRerouteUsage()
  n150 = Add(A=Add(A=Multiply(A=@n144, B=@n145), B=Multiply(A=@n146, B=@n147)), B=Multiply(A=Multiply(A=@n148, B=@n149), B=NamedRerouteUsage()))
  n151 = Constant3Vector() {constant=<rgba 0 1 0 0>}
  n152 = NamedRerouteUsage()
  n153 = Multiply(
   A=LinearInterpolate(A=Multiply(A=@n151, B=@n152), B=Constant3Vector() {constant=<rgba 0.1 0 0 0>}, Alpha=NamedRerouteUsage())
   B=NamedRerouteUsage()
  )
  NamedRerouteDeclaration(None=Multiply(A=Add(A=@n150, B=@n153), B=NamedRerouteUsage())) {name=Axis_Legs, nodeColor=<rgba 0 0.3437 1 1>}
  n154 = Time()
  n155 = Constant() {r=2}
  NamedRerouteDeclaration(None=Add(A=Divide(A=@n154, B=@n155), B=@n100)) {name=Time_Primary_Idle/Loop, nodeColor=<rgba 0.7578 1 0 1>}
  n156 = Step(X=@n135) {constX=0.9, constY=0.9}
  NamedRerouteDeclaration(None=Subtract(A=Step(X=@n135) {constX=0.9, constY=0.7}, B=NamedRerouteDeclaration(None=@n156) {name=M_Legs_IK_1, nodeColor=<rgba 0.5469 1 0 1>})) {name=M_Legs_IK_2, nodeColor=<rgba 0.5469 1 0 1>}
  NamedRerouteDeclaration(None=Multiply(A=Constant3Vector() {constant=<rgba 0 0.5 1 0>}, B=NamedRerouteUsage())) {name=Axis_Mouth, nodeColor=<rgba 0.6641 1 0 1>}
  NamedRerouteDeclaration(None=Divide(A=Constant() {desc="Curve Index (0.5,1.5,2.5...)", r=2.5}, B=@n54)) {name=CA_ROW_03_Wings, nodeColor=<rgba 0 1 0.5243 1>}
  NamedRerouteDeclaration(None=Divide(A=Constant() {desc="Curve Index (0.5,1.5,2.5...)", r=5.5}, B=@n54)) {name=CA_ROW_05_Antlers, nodeColor=<rgba 1 0.6563 0.0149 1>}
  n157 = Multiply(A=Constant3Vector() {constant=<rgba 80 75 70 0>}, B=NamedRerouteUsage())
  n158 = Add(
   A=Multiply(A=Constant3Vector() {constant=<rgba 110 25 70 0>}, B=NamedRerouteUsage())
   B=Multiply(A=Constant3Vector() {constant=<rgba 78 -165 96 0>}, B=NamedRerouteUsage())
  )
  NamedRerouteDeclaration(None=Multiply(A=Add(A=@n157, B=@n158), B=NamedRerouteUsage())) {name=Legs_Pivot Point_B, nodeColor=<rgba 1 0.7734 0 1>}
  n159 = Multiply(A=NamedRerouteUsage()) {constB=-0.1}
  n160 = AppendVector(A=Add(A=NamedRerouteUsage(), B=@n159), B=NamedRerouteUsage())
  NamedRerouteDeclaration(None=TextureSample(UVs=Frac(None=@n160), Tex=NamedRerouteUsage()) {samplerType=SAMPLERTYPE_LinearColor, texture=&CA_SwarmerWPOCurves}.G) {name=Curve_Antlers_Walk_Rot1_Delay, desc=legs torso, nodeColor=<rgba 1 0 0.6562 1>}
  n161 = NamedRerouteUsage()
  n162 = Add(A=Multiply(A=NamedRerouteUsage()) {constB=-0.1}, B=Multiply(A=NamedRerouteUsage()) {constB=0.4})
  n163 = AppendVector(A=Add(A=Add(A=@n161, B=@n162)) {constB=0.25}, B=NamedRerouteUsage())
  NamedRerouteDeclaration(None=TextureSample(UVs=Frac(None=@n163), Tex=NamedRerouteUsage()) {samplerType=SAMPLERTYPE_LinearColor, texture=&CA_SwarmerWPOCurves}.R) {name=Curve_Wings_Loop_Rot1, nodeColor=<rgba 1 0 0.0807 1>}
  n164 = LocalPosition() {includedOffsets=ExcludeOffsets, localOrigin=InstancePreSkinning}
  NamedRerouteDeclaration(None=ShaderStageSwitch(PixelShader=VertexInterpolator(VS=@n164.XYZ).PS, VertexShader=@n164.XYZ)) {name=Pos_Preskined, nodeColor=<rgba 0 0.9766 1 1>}
  n165 = NamedRerouteUsage()
  n166 = Constant3Vector() {constant=<rgba 0 -40 0 0>}
  n167 = NamedRerouteUsage()
  n168 = NamedRerouteUsage()
  NamedRerouteDeclaration(None=Multiply(A=Multiply(A=@n165, B=@n166), B=Max(A=@n167, B=@n168))) {name=Pose_Body, nodeColor=<rgba 1 0.9141 0 1>}
  n169 = MaterialFunctionCall(Color=NamedRerouteUsage()) {materialFunction=&SplitComponents}
  NamedRerouteDeclaration(None=MaterialFunctionCall(R=Abs(None=@n169.R), G=@n169.G, B=@n169.G) {materialFunction=&AppendMany}.RGB) {name=Pos_Preskinned_AbsR, nodeColor=<rgba 0 0.3672 1 1>}
  n170 = NamedRerouteDeclaration(None=Sign(None=@n52)) {name="M_Left/Right_(-1,1)", nodeColor=<rgba 0 0.0625 1 1>}
  n171 = Constant() {r=1}
  NamedRerouteDeclaration(None=MaterialFunctionCall(R=@n170, G=@n171, B=@n171) {materialFunction=&AppendMany}.RGB) {name="M_Left/Right(-1,1)_Red", nodeColor=<rgba 0.5938 0 1 1>}
  n172 = NamedRerouteUsage()
  n173 = NamedRerouteUsage()
  n174 = Add(A=Add(A=Add(A=@n172, B=@n173), B=NamedRerouteUsage()), B=NamedRerouteUsage())
  n175 = Add(A=NamedRerouteUsage(), B=NamedRerouteUsage())
  n176 = NamedRerouteUsage()
  n177 = Add(A=NamedRerouteUsage(), B=NamedRerouteUsage())
  n178 = NamedRerouteUsage()
  n179 = MaterialFunctionCall!(
   RotationAngle=Add(A=Add(A=@n175, B=@n176), B=NamedRerouteUsage())
   NormalizedRotationAxis=Normalize(VectorInput=Add(A=@n174) {constB=0, desc=SafeNormalize})
   PivotPoint=Add(A=Add(A=@n177, B=@n178), B=NamedRerouteUsage())
   Position=NamedRerouteUsage()
   Normal
   TangentU
   TangentV
  ) {materialFunction=/Game/Characters/Player/Male/Medium/Bodies/M_MED_Inferno/FX/Mat/RotateVertexAboutAxis}
  NamedRerouteDeclaration(None=LinearInterpolate(A=VertexNormalWS(), B=@n179.Normal, Alpha=NamedRerouteUsage())) {name=N_Rotations1, nodeColor=<rgba 0 1 0.2734 1>}
  n180 = NamedRerouteUsage()
  n181 = NamedRerouteUsage()
  n182 = NamedRerouteUsage()
  n183 = MaterialFunctionCall(Float4=Add(A=Multiply(A=@n73, B=@n181), B=Multiply(A=@n72, B=@n182))) {materialFunction=&BreakOutFloat4Components}
  n184 = NamedRerouteUsage()
  n185 = StaticSwitch(True=NamedRerouteUsage(), False=@n183.B, Value=NamedRerouteUsage())
  NamedRerouteDeclaration(None=Add(
    A=StaticSwitch(True=@n180, False=@n183.G, Value=@n184)
    B=NamedRerouteDeclaration(None=@n185) {name=Anim_2_Sprint, nodeColor=<rgba 1 0.4129 0.2165 1>}
   )) {name=Anim_1_Walk, nodeColor=<rgba 0.2457 0.7088 1 1>}
  NamedRerouteDeclaration(None=Divide(A=ComponentMask(None=@n51) {b=true}) {constB=250}) {name=M_VerticalGradient, nodeColor=<rgba 1 0.4453 0 1>}
  NamedRerouteDeclaration(None=@n39.R) {name=PD_GameTime, nodeColor=<rgba 0 0.9297 1 1>}
  n186 = AppendVector(A=@n161, B=NamedRerouteUsage())
  NamedRerouteDeclaration(None=TextureSample(UVs=Frac(None=@n186), Tex=NamedRerouteUsage()) {samplerType=SAMPLERTYPE_LinearColor, texture=&CA_SwarmerWPOCurves}.A) {name=Curve_Wings_Loop_Rot1_Fast, nodeColor=<rgba 1 0 0.0807 1>}
  n187 = LinearInterpolate(A=NamedRerouteUsage(), B=NamedRerouteUsage(), Alpha=NamedRerouteUsage()) {constB=0}
  n188 = Constant() {r=-0.005}
  n189 = Add(A=Multiply(A=NamedRerouteUsage(), B=Constant() {r=0.02}) {constB=0}, B=Constant() {r=-0.035})
  n190 = NamedRerouteUsage()
  n191 = Constant() {r=0.005}
  n192 = Constant() {r=0.01}
  n193 = NamedRerouteUsage()
  n194 = Add(
   A=LinearInterpolate(A=LinearInterpolate(A=@n188, B=@n189, Alpha=@n190), B=LinearInterpolate(A=@n191, B=@n192, Alpha=@n193), Alpha=NamedRerouteUsage())
   B=Constant() {r=0.027}
  )
  n195 = LinearInterpolate(Alpha=Max(A=NamedRerouteUsage(), B=NamedRerouteUsage())) {constA=-0.05, constB=0.1}
  n196 = NamedRerouteUsage()
  n197 = LinearInterpolate(A=Add(A=Multiply(A=@n187, B=@n194) {constB=0.01}, B=Multiply(A=@n195, B=@n196)), B=NamedRerouteUsage(), Alpha=NamedRerouteUsage())
  n198 = LinearInterpolate(A=Add(A=Multiply(A=@n197) {constB=0.1}) {constB=-0.1}, Alpha=NamedRerouteUsage()) {constA=-0.1, constB=-0.25}
  n199 = NamedRerouteUsage()
  n200 = NamedRerouteUsage()
  n201 = LinearInterpolate(A=Constant3Vector() {constant=<rgba 0 1 0 0>}, B=Constant3Vector() {constant=<rgba 0 0 1 0>}, Alpha=NamedRerouteUsage())
  n202 = Constant3Vector() {constant=<rgba 1 0 0 0>}
  n203 = Max(A=Max(A=NamedRerouteUsage(), B=NamedRerouteUsage()), B=NamedRerouteUsage())
  n204 = MaterialFunctionCall(
   RotationAngle=LinearInterpolate(
    A=LinearInterpolate(A=@n197, B=@n198, Alpha=@n199) {constB=-0.1}
    B=LinearInterpolate(Alpha=@n200) {constA=-0.1, constB=-0.05}
    Alpha=NamedRerouteUsage()
   ) {constB=-0.1}
   NormalizedRotationAxis=Normalize(VectorInput=LinearInterpolate(A=@n201, B=@n202, Alpha=@n203) {constB=0.2})
   PivotPoint=Constant3Vector() {constant=<rgba 0 3 80 0>}
   Position=Add(A=NamedRerouteUsage(), B=NamedRerouteUsage())
   Normal=NamedRerouteUsage()
  ) {materialFunction=&RotateVertexAboutAxis}
  NamedRerouteDeclaration(None=@n204.Normal) {name=N_Rotations2, nodeColor=<rgba 1 0.4219 0 1>}
  n205 = Multiply(A=NamedRerouteUsage(), B=Constant() {r=-30})
  n206 = Multiply(A=NamedRerouteUsage(), B=Constant() {r=-40})
  n207 = NamedRerouteUsage()
  n208 = Constant() {r=-10}
  n209 = MaterialFunctionCall(R=Multiply(A=NamedRerouteUsage(), B=Constant() {r=-30}), G=Add(A=Add(A=@n205, B=@n206), B=Multiply(A=@n207, B=@n208)), B=Constant()) {materialFunction=&AppendMany}
  n210 = Multiply(A=Multiply(A=NamedRerouteUsage(), B=Constant3Vector() {constant=<rgba 20 -85 50 0>}), B=NamedRerouteUsage())
  n211 = Multiply(A=NamedRerouteUsage(), B=Constant3Vector() {constant=<rgba 0 0 50 0>})
  n212 = NamedRerouteUsage()
  n213 = Multiply(A=Add(A=Add(A=@n209.RGB, B=@n210), B=Multiply(A=@n211, B=@n212)), B=NamedRerouteUsage())
  n214 = Multiply(A=LocalPosition().XYZ) {constB=-0.8}
  n215 = NamedRerouteUsage()
  NamedRerouteDeclaration(None=Multiply(A=LinearInterpolate(A=@n213, B=@n214, Alpha=@n215) {constB=0}, B=NamedRerouteUsage())) {name=Pose_Legs, nodeColor=<rgba 0.8281 0 1 1>}
  NamedRerouteDeclaration(None=Multiply(A=@n136, B=NamedRerouteUsage())) {name=M_MouthGradient, nodeColor=<rgba 0.6406 0 1 1>}
  n216 = NamedRerouteDeclaration(None=Multiply(A=@n105.B, B=NamedRerouteUsage())) {name=M_Head, nodeColor=<rgba 0 1 0.3438 1>}
  n217 = Saturate(None=Subtract(A=Multiply(A=@n136, B=@n216), B=NamedRerouteUsage()))
  NamedRerouteDeclaration(None=Step(X=NamedRerouteDeclaration(None=@n217) {name=M_AntlersGrad, nodeColor=<rgba 0.8047 0 1 1>}) {constY=0.1}) {name=M_Antlers, nodeColor=<rgba 1 0.9141 0 1>}
  n218 = NamedRerouteUsage()
  n219 = NamedRerouteUsage()
  NamedRerouteDeclaration(None=Add(A=Add(A=@n218, B=@n219), B=NamedRerouteUsage())) {name=Pre-AnimPose, nodeColor=<rgba 0 1 0.5078 1>}
  n220 = NamedRerouteUsage()
  n221 = Constant4Vector() {desc="0 - Idle\r\n1 - Walk\r\n2 - MeleeAttack\r\n3- Falling\r\n4 - Death\r\n5 - Flower\r\n6 - Jump", constant=<rgba 0 1 2 3>}
  NamedRerouteDeclaration(None=Multiply(A=Step(Y=@n221.RGBA, X=@n220), B=Step(Y=@n220, X=@n221.RGBA))) {name=IndexA_Validation 1_1, nodeColor=<rgba 0.0781 1 0 1>}
  n222 = Constant() {r=0.08}
  n223 = Multiply(A=NamedRerouteUsage(), B=Constant() {r=0.15}) {constB=0}
  NamedRerouteDeclaration(None=Multiply(A=Add(A=@n222, B=@n223), B=NamedRerouteUsage())) {name=Rot_Body_MeleeAttackCombo, nodeColor=<rgba 0 1 0.9531 1>}
  n224 = LinearInterpolate(A=NamedRerouteUsage(), B=NamedRerouteUsage(), Alpha=NamedRerouteUsage())
  n225 = LinearInterpolate(Alpha=NamedRerouteUsage()) {constA=0.02, constB=0.05}
  NamedRerouteDeclaration(None=Multiply(A=Multiply(A=@n224, B=@n225) {constB=0.05}, B=NamedRerouteUsage()) {constB=0}) {name=Rot_Antlers, nodeColor=<rgba 0 1 0.3672 1>}
  n226 = NamedRerouteUsage()
  n227 = Constant3Vector() {constant=<rgba -5 0 -20 0>}
  NamedRerouteDeclaration(None=Multiply(A=NamedRerouteUsage(), B=Multiply(A=@n226, B=@n227))) {name=Pose_Antlers, nodeColor=<rgba 0 1 0.2031 1>}
  NamedRerouteDeclaration(None=StaticSwitch(True=NamedRerouteUsage(), False=@n39.G, Value=NamedRerouteUsage())) {name="PD_A>B/B>A", nodeColor=<rgba 1 0.75 0 1>}
  NamedRerouteDeclaration(None=MaterialFunctionCall(R=@n170, G=@n170, B=Constant() {r=1}) {materialFunction=&AppendMany}.RGB) {name="M_Left/Right(-1,1)_RG", nodeColor=<rgba 0.5938 0 1 1>}
  n228 = Add(A=NamedRerouteUsage())
  n229 = Constant()
  n230 = Constant() {r=45}
  n231 = Constant() {r=40}
  n232 = Constant()
  n233 = Constant() {r=10}
  n234 = Constant() {r=5}
  n235 = LinearInterpolate(
   A=Multiply(
    A=@n228
    B=MaterialFunctionCall(R=@n229, G=@n230, B=@n231) {materialFunction=&AppendMany}.RGB
   )
   B=Multiply(
    A=@n228
    B=MaterialFunctionCall(R=@n232, G=@n233, B=@n234) {materialFunction=&AppendMany}.RGB
   ) {constB=-1}
   Alpha=NamedRerouteUsage()
  )
  n236 = Constant() {r=3}
  n237 = Constant() {r=10}
  n238 = Constant() {r=-20}
  n239 = Multiply(
   A=@n228
   B=LinearInterpolate(
    A=MaterialFunctionCall(R=@n236, G=@n237, B=@n238) {materialFunction=&AppendMany}.RGB
    Alpha=NamedRerouteUsage()
   )
  )
  n240 = NamedRerouteUsage()
  n241 = Multiply(A=Multiply(A=LinearInterpolate(A=@n235, B=@n239, Alpha=@n240), B=NamedRerouteUsage()), B=NamedRerouteUsage())
  n242 = NamedRerouteUsage()
  n243 = LinearInterpolate(Alpha=NamedRerouteUsage()) {constA=1, constB=0}
  n244 = Multiply(A=Saturate(None=NamedRerouteUsage()), B=Constant3Vector() {constant=<rgba 0 0 1 0>})
  n245 = Multiply(A=NamedRerouteUsage(), B=Constant3Vector() {constant=<rgba -1 -1 0 0>})
  n246 = Constant()
  n247 = NamedRerouteUsage()
  n248 = Constant() {r=60}
  n249 = NamedRerouteUsage()
  n250 = Constant() {r=55}
  n251 = Add(A=Add(A=Multiply(A=@n247, B=@n248), B=Multiply(A=@n249, B=@n250)), B=Multiply(A=NamedRerouteUsage(), B=Constant() {r=50}))
  n252 = LinearInterpolate(A=Constant() {r=30}, B=Constant() {r=45}, Alpha=NamedRerouteUsage())
  n253 = Multiply(
   A=Multiply(
    A=Add(A=@n244, B=@n245)
    B=MaterialFunctionCall(R=@n246, G=@n251, B=@n252) {materialFunction=&AppendMany}.RGB
   )
   B=NamedRerouteUsage()
  )
  n254 = NamedRerouteUsage()
  n255 = MaterialFunctionCall(R=Constant(), G=Multiply(A=@n254, B=Constant() {r=50}), B=Multiply(A=@n254, B=Constant() {r=-30})) {materialFunction=&AppendMany}
  n256 = NamedRerouteUsage()
  n257 = LinearInterpolate(
   A=LinearInterpolate(A=@n253, B=Multiply(A=@n253, B=@n243), Alpha=NamedRerouteUsage()) {constB=0}
   B=Multiply(A=Multiply(A=@n255.RGB, B=@n256), B=NamedRerouteUsage())
   Alpha=NamedRerouteUsage()
  )
  n258 = NamedRerouteUsage()
  n259 = Add(A=LinearInterpolate(A=Multiply(A=@n257, B=@n258), Alpha=NamedRerouteUsage()) {constB=0}, B=@n179.WPO) {constB=0}
  n260 = LinearInterpolate(A=@n259, B=Add(A=@n259, B=Multiply(A=@n241, B=@n242)), Alpha=NamedRerouteUsage())
  NamedRerouteDeclaration(None=LinearInterpolate(A=@n260, B=Multiply(A=@n260) {constB=0.1}, Alpha=NamedRerouteUsage()) {constB=0}) {name=Anim_LegsMain+SecondaryTorso, nodeColor=<rgba 1 0.1875 0 1>}
  NamedRerouteDeclaration(None=Multiply(A=Constant3Vector() {constant=<rgba 0 25 80 0>}, B=NamedRerouteUsage())) {name=Head_PivotPoint, nodeColor=<rgba 0.5469 1 0 1>}
  n261 = LinearInterpolate(A=Constant3Vector() {constant=<rgba 32 10 82 0>}, B=Constant3Vector() {constant=<rgba 10 19 105 0>}, Alpha=NamedRerouteUsage())
  n262 = NamedRerouteUsage()
  NamedRerouteDeclaration(None=Multiply(A=Multiply(A=@n261, B=@n262), B=NamedRerouteUsage())) {name=Wings_PivotPoint, nodeColor=<rgba 0.1016 0 1 1>}
  n263 = Multiply(A=@n108, B=@n105.R)
  NamedRerouteDeclaration(None=Step(X=NamedRerouteDeclaration(None=@n263) {name=M_WingsValues, nodeColor=<rgba 0 1 0.1562 1>}) {constX=0.5, constY=0.7}) {name=M_WingsTop, nodeColor=<rgba 0.2187 1 0 1>}
  n264 = LinearInterpolate(
   A=LinearInterpolate(Alpha=NamedRerouteUsage()) {constA=1, constB=0.8}
   B=LinearInterpolate(Alpha=NamedRerouteUsage()) {constB=1.5}
   Alpha=NamedRerouteUsage()
  )
  n265 = LinearInterpolate(Alpha=NamedRerouteUsage()) {constB=0.8}
  n266 = NamedRerouteUsage()
  n267 = NamedRerouteUsage()
  n268 = NamedRerouteUsage()
  n269 = NamedRerouteUsage()
  n270 = Max(A=NamedRerouteUsage(), B=NamedRerouteUsage())
  n271 = NamedRerouteUsage()
  n272 = Add(
   A=LinearInterpolate(A=LinearInterpolate(A=@n267, B=@n268, Alpha=@n269), B=NamedRerouteUsage(), Alpha=Max(A=@n270, B=@n271))
   B=LinearInterpolate(A=LinearInterpolate(A=@n264, B=@n265, Alpha=@n266), Alpha=NamedRerouteUsage()) {constB=4}
  ) {desc=Rotation Offset}
  n273 = LinearInterpolate(
   A=LinearInterpolate(Alpha=NamedRerouteUsage()) {constA=50, constB=12.5}
   B=LinearInterpolate(Alpha=NamedRerouteUsage()) {constA=30, constB=12.5}
   Alpha=NamedRerouteUsage()
  )
  NamedRerouteDeclaration(None=Multiply(A=Divide(A=@n272, B=@n273) {desc=Rotation Range}, B=NamedRerouteUsage())) {name=Rot_Wings, nodeColor=<rgba 1 0 0.5391 1>}
  n274 = LinearInterpolate(A=Constant3Vector() {constant=<rgba 0 -3 -1 0>}, B=Constant3Vector() {constant=<rgba 0.7 -0.25 -1 0>}, Alpha=NamedRerouteUsage())
  n275 = NamedRerouteUsage()
  n276 = MaterialFunctionCall(R=Constant() {r=1}, G=@n275, B=@n275) {materialFunction=&AppendMany}
  NamedRerouteDeclaration(None=Multiply(A=Multiply(A=@n274, B=@n276.RGB), B=NamedRerouteUsage())) {name=A_R1_Wings_Axis, nodeColor=<rgba 0 0.8594 1 1>}
  NamedRerouteDeclaration(None=StaticSwitch(True=NamedRerouteUsage(), False=@n183.R, Value=NamedRerouteUsage())) {name=Anim_0_Idle, nodeColor=<rgba 0.4809 0.1848 1 1>}
  n277 = AppendVector(A=Add(A=NamedRerouteUsage(), B=@n159), B=NamedRerouteUsage())
  NamedRerouteDeclaration(None=TextureSample(UVs=Frac(None=@n277), Tex=NamedRerouteUsage()) {samplerType=SAMPLERTYPE_LinearColor, texture=&CA_SwarmerWPOCurves}.G) {name=Curve_Antlers_Idle_Rot1_Delay, desc=legs torso, nodeColor=<rgba 1 0 0.6562 1>}
  n278 = NamedRerouteUsage()
  n279 = Divide(A=Time(), B=Constant() {r=0.7})
  n280 = NamedRerouteUsage()
  NamedRerouteDeclaration(None=Add(A=StaticSwitch(True=@n278, False=@n279, Value=@n280), B=@n100)) {name=Time_Primary_Walk, nodeColor=<rgba 0.7578 1 0 1>}
  n281 = NamedRerouteUsage()
  n282 = NamedRerouteUsage()
  n283 = TextureSample(UVs=Frac(None=AppendVector(A=@n281, B=@n282)), Tex=NamedRerouteUsage()) {samplerType=SAMPLERTYPE_LinearColor, texture=&CA_SwarmerWPOCurves}
  n284 = NamedRerouteUsage()
  n285 = NamedRerouteUsage()
  n286 = TextureSample(UVs=Frac(None=AppendVector(A=@n284, B=@n285)), Tex=NamedRerouteUsage()) {samplerType=SAMPLERTYPE_LinearColor, texture=&CA_SwarmerWPOCurves}
  n287 = NamedRerouteUsage()
  n288 = NamedRerouteUsage()
  n289 = NamedRerouteUsage()
  NamedRerouteDeclaration(None=LinearInterpolate(A=LinearInterpolate(A=@n283.R, B=@n286.R, Alpha=@n287), B=@n283.B, Alpha=Max(A=@n288, B=@n289))) {name=Curve_Mouth_Idle_Rot1, desc=legs torso, nodeColor=<rgba 1 0 0.6562 1>}
  n290 = Constant3Vector() {constant=<rgba 20 50 76 0>}
  n291 = NamedRerouteUsage()
  NamedRerouteDeclaration(None=Multiply(A=Multiply(A=@n290, B=@n291), B=NamedRerouteUsage())) {name=Mouth_Pivot, nodeColor=<rgba 0.5469 1 0 1>}
  n292 = LinearInterpolate(A=NamedRerouteUsage(), B=NamedRerouteUsage(), Alpha=NamedRerouteUsage())
  n293 = LinearInterpolate(A=@n292, B=LinearInterpolate(A=@n292, B=NamedRerouteUsage(), Alpha=NamedRerouteUsage()), Alpha=NamedRerouteUsage())
  n294 = NamedRerouteUsage()
  n295 = NamedRerouteUsage()
  NamedRerouteDeclaration(None=LinearInterpolate(A=LinearInterpolate(A=@n293, B=@n294, Alpha=@n295), B=NamedRerouteUsage(), Alpha=NamedRerouteUsage())) {name=Legs_PivotPoint, nodeColor=<rgba 0.2891 1 0 1>}
  n296 = NamedRerouteUsage()
  n297 = NamedRerouteUsage()
  NamedRerouteDeclaration(None=LinearInterpolate(A=Multiply(A=@n296, B=@n297) {constB=0.01}, Alpha=NamedRerouteUsage()) {constB=0}) {name=Anim_TorsoTranversal, nodeColor=<rgba 0 0.1328 1 1>}
  n298 = NamedRerouteUsage()
  n299 = NamedRerouteUsage()
  n300 = NamedRerouteUsage()
  NamedRerouteDeclaration(None=Multiply(A=Add(A=@n204.WPO, B=@n298), B=Max(A=@n299, B=@n300))) {name=Anim_TorsoMain+LegsSecondary, nodeColor=<rgba 0.4531 1 0 1>}
  NamedRerouteDeclaration(None=Multiply(A=Constant3Vector() {constant=<rgba 1 0 0 0>}, B=NamedRerouteUsage())) {name=Axis_Antlers, nodeColor=<rgba 1 0.0469 0 1>}
  n301 = NamedRerouteUsage()
  n302 = Constant4Vector() {desc="0 - Idle\r\n1 - Walk\r\n2 - MeleeAttack\r\n3- Falling\r\n4 - Death\r\n5 - Flower\r\n6 - Jump", constant=<rgba 4 5 6 7>}
  NamedRerouteDeclaration(None=Multiply(A=Step(Y=@n302.RGBA, X=@n301), B=Step(Y=@n301, X=@n302.RGBA))) {name=IndexA_Validation 2_1, nodeColor=<rgba 0.0781 1 0 1>}
  NamedRerouteDeclaration(None=Multiply(A=Constant3Vector() {constant=<rgba 0 45 160 0>}, B=NamedRerouteUsage())) {name=Pivot_Antlers, nodeColor=<rgba 0.5937 1 0 1>}
  n303 = NamedRerouteUsage()
  NamedRerouteDeclaration(None=Multiply(A=Step(Y=@n221.RGBA, X=@n303), B=Step(Y=@n303, X=@n221.RGBA))) {name=IndexB_Validation 1_1, nodeColor=<rgba 0.4429 0 1 1>}
  n304 = NamedRerouteUsage()
  n305 = NamedRerouteDeclaration(None=Multiply(A=NamedRerouteUsage(), B=NamedRerouteUsage())) {name=M_LegsFrontGrad, nodeColor=<rgba 1 0 0.6094 1>}
  n306 = Max(A=@n304, B=@n305)
  n307 = NamedRerouteUsage()
  n308 = NamedRerouteUsage()
  NamedRerouteDeclaration(None=LinearInterpolate(A=@n304, B=LinearInterpolate(A=@n304, B=@n306, Alpha=@n305), Alpha=Add(A=@n307, B=@n308)) {constB=0}) {name=M_Torso, nodeColor=<rgba 0 1 0.625 1>}
  n309 = Add(A=VertexNormalWS(), B=Constant3Vector() {constant=<rgba 0 -0.5 1.5 0>})
  n310 = Constant3Vector() {constant=<rgba -13 0 -10 0>}
  n311 = Multiply(A=Multiply(A=Multiply(A=@n309, B=@n310) {constB=-5}, B=NamedRerouteUsage()), B=Add(A=NamedRerouteUsage())) {constB=0}
  n312 = NamedRerouteUsage()
  n313 = NamedRerouteUsage()
  n314 = Constant3Vector() {constant=<rgba 5 5 -1 0>}
  n315 = Multiply(A=@n309, B=Add(A=Constant3Vector() {constant=<rgba -15 -15 30 0>}, B=Multiply(A=@n313, B=@n314) {constB=-0.1})) {constB=-5}
  n316 = NamedRerouteUsage()
  NamedRerouteDeclaration(None=LinearInterpolate(A=Multiply(A=@n311, B=@n312), B=Multiply(A=@n315, B=@n316), Alpha=NamedRerouteUsage()) {constB=0.1}) {name=Abdomen_Inflate, nodeColor=<rgba 0 0.5547 1 1>}
  n317 = NamedRerouteUsage()
  NamedRerouteDeclaration(None=Multiply(A=Step(Y=@n302.RGBA, X=@n317), B=Step(Y=@n317, X=@n302.RGBA))) {name=IndexB_Validation 2_1, nodeColor=<rgba 0.4429 0 1 1>}
  NamedRerouteDeclaration(None=StaticBoolParameter() {group=WPO, parameterName=Is VFX, sortPriority=-10}) {name=Is_VFX, nodeColor=<rgba 1 0.0703 0 1>}
  n318 = MaterialFunctionCall(R=NamedRerouteUsage(), G=NamedRerouteUsage(), B=NamedRerouteUsage()) {materialFunction=&AppendMany}
  n319 = NamedRerouteUsage()
  n320 = MaterialFunctionCall(R=NamedRerouteUsage(), G=@n319, B=@n319) {materialFunction=&AppendMany}
  n321 = NamedRerouteUsage()
  NamedRerouteDeclaration(None=LinearInterpolate(A=LinearInterpolate(A=@n318.RGB, B=@n320.RGB, Alpha=@n321), Alpha=NamedRerouteUsage())) {name=Sine_Transversal2_Torso, nodeColor=<rgba 0.9219 1 0 1>}
  n322 = AppendVector(A=NamedRerouteUsage(), B=NamedRerouteUsage())
  NamedRerouteDeclaration(None=TextureSample(UVs=Frac(None=@n322), Tex=NamedRerouteUsage()) {samplerType=SAMPLERTYPE_LinearColor, texture=&CA_SwarmerWPOCurves}.B) {name=Curve_Wings_Attack_Rot1, nodeColor=<rgba 1 0 0.0807 1>}
  n323 = NamedRerouteUsage()
  n324 = Multiply(A=@n323, B=NamedRerouteUsage())
  n325 = Multiply(A=@n323, B=NamedRerouteUsage())
  NamedRerouteDeclaration(None=Add(
    A=NamedRerouteDeclaration(None=@n324) {name=M_Wings_Gradient, nodeColor=<rgba 1 0 0.7734 1>}
    B=NamedRerouteDeclaration(None=@n325) {name=M_Abdomen_Gradient, nodeColor=<rgba 1 0 0.7734 1>}
   )) {name=M_Wings+Abdomen Gradient, nodeColor=<rgba 0 0.7422 1 1>}
  NamedRerouteDeclaration(None=StaticSwitch(True=NamedRerouteUsage(), False=@n76.G, Value=NamedRerouteUsage())) {name=Anim_5_Falling, nodeColor=<rgba 0.8498 1 0.2318 1>}
  NamedRerouteDeclaration(None=StaticSwitch(True=NamedRerouteUsage(), False=@n76.A, Value=NamedRerouteUsage())) {name=Anim_7_Death, nodeColor=<rgba 0.0576 0.0712 0.0156 1>}
  NamedRerouteDeclaration(None=StaticSwitch(True=NamedRerouteUsage(), False=@n76.B, Value=NamedRerouteUsage())) {name=Anim_6_Flower, nodeColor=<rgba 0.128 0.0135 0.2274 1>}
  NamedRerouteDeclaration(None=StaticSwitch(True=NamedRerouteUsage(), False=@n183.A, Value=NamedRerouteUsage())) {name=Anim_3_Fly, nodeColor=<rgba 0.358 1 0.5501 1>}
  n326 = Multiply(A=NamedRerouteUsage(), B=Constant3Vector() {desc="20,8,8", constant=<rgba -10 0 5 1>}) {constB=0.2}
  n327 = NamedRerouteUsage()
  n328 = Constant() {r=-10}
  n329 = Multiply(
   A=MaterialFunctionCall(R=Multiply(A=@n327, B=@n328), G=Constant(), B=Constant() {r=-10}) {materialFunction=&AppendMany}.RGB
   B=NamedRerouteUsage()
  )
  NamedRerouteDeclaration(None=Add(A=Add(A=@n326, B=@n329), B=Constant3Vector() {desc="5,5,4", constant=<rgba 5 5 4 1>})) {name=Anim_T2_Range, nodeColor=<rgba 0.4531 1 0 1>}
  NamedRerouteUsage()
  Saturate(None=MaterialFunctionCall(Number=@n21) {materialFunction=&DebugScalarValues}.Result)
  n330 = Constant() {r=1}
  Saturate(None=MaterialFunctionCall(Number=Fmod(A=@n6, B=@n330)) {materialFunction=&DebugScalarValues}.Result)
  ScalarParameter() {group=WPO, parameterName="!!!!!! Turn off the above !!!!!!!", sortPriority=0}
  ScalarParameter() {defaultValue=-1}
  n331 = NamedRerouteUsage()
  n332 = Constant3Vector() {constant=<rgba 0 0.9387 1 1>}
  n333 = NamedRerouteUsage()
  n334 = LinearInterpolate(
   A=LinearInterpolate(A=LinearInterpolate(A=@n331, B=@n332, Alpha=@n333), B=Constant3Vector() {constant=<rgba 1 0.0232 0 1>}, Alpha=NamedRerouteUsage())
   B=Constant3Vector() {constant=<rgba 0.0252 0 1 1>}
   Alpha=NamedRerouteUsage()
  )
  n335 = NamedRerouteUsage()
  n336 = LinearInterpolate(
   A=LinearInterpolate(A=LinearInterpolate(A=@n334, Alpha=@n335) {constB=0}, B=Constant3Vector() {constant=<rgba 1 0 0.8632 1>}, Alpha=NamedRerouteUsage())
   B=Constant3Vector() {constant=<rgba 0.0976 1 0 1>}
   Alpha=NamedRerouteUsage()
  )
  n337 = VectorParameter() {group=WPO_Bake, parameterName=Base Color, defaultValue=<rgba 0.0789 0.0649 0.1267 1>}
  n338 = Constant()
  SetMaterialAttributes!(
   MaterialAttributes
   Base Color=StaticSwitch(
    True=ParticleColor().RGB
    False=StaticSwitchParameter(True=@n336, False=@n337.RGB) {group=WPO, parameterName=Debug_AnimStateColors, sortPriority=-1}
    Value=NamedRerouteUsage()
   )
   Metallic=@n338
   Specular=@n338
   Roughness=@n338
   Emissive Color
   World Position Offset=NamedRerouteUsage()
   Normal=NamedRerouteUsage()
  ) {attributeSetTypes [
    #1
    #2
    #3
    #4
    #5
    #6
    #7
   ]}
  ShaderStageSwitch(PixelShader=VertexInterpolator(VS=@n18).PS, VertexShader=@n18)
  ShaderStageSwitch(PixelShader=VertexInterpolator(VS=@n15).PS, VertexShader=@n15)
  Sine(None=@n3)
  StaticSwitch(True=NamedRerouteUsage(), False=@n83, Value=NamedRerouteUsage())
  StaticSwitch(True=NamedRerouteUsage(), False=@n78, Value=NamedRerouteUsage())
  n339 = TextureCoordinate() {coordinateIndex=3}
  Step(Y=Distance(A=@n339, B=Constant2Vector() {g=0.6})) {constX=0.1, constY=0.1, desc=abdomen}
  Step(Y=Distance(A=@n339, B=Constant2Vector() {g=0.4, r=0.6})) {constX=0.1, constY=0.1, desc=wings}
 ]
}
