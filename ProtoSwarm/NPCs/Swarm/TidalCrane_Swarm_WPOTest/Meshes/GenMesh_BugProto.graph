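# Editor utility mesh actor that builds the LODs of SM_BugProto_Generated from SM_BugProto_TA2 for the swarm bug prototype.
# RunMeshGen releases the three components' compute meshes, then copies the source into fresh pooled meshes using Geometry Script.
# It writes an unsimplified LOD0 and two progressively vertex-reduced LODs, each with its own simplification method variable.
# Each successfully written mesh is also set on its own DynamicMeshComponent.
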EditorUtilityBlueprint /ProtoSwarm/NPCs/Swarm/TidalCrane_Swarm_WPOTest/Meshes/GenMesh_BugProto
graphs [
 EventGraph {nodes [
   Event BeginPlay()
   Event ActorBeginOverlap()
   Event Tick!() -> (OutputDelegate, then, DeltaSeconds:float)
  ]}
 NewFunction {nodes [NewFunction() {is=title K2Node_FunctionEntry}]}
 RunMeshGen {nodes [
   RunMeshGen() {is=title K2Node_FunctionEntry}
   > n1 = Sequence!(execute) -> (then_0, then_1, then_2:Exec, then_3:Exec)
   > n2 = GetDynamicMesh!(
    execute
    self:Base Dynamic Mesh Component Object Reference=Get DynamicMeshComponent() -> (DynamicMeshComponent:Dynamic Mesh Component Object Reference)
   ) -> (then, ReturnValue:Dynamic Mesh Object Reference)
   > ReleaseComputeMesh!(execute, self:Dynamic Mesh Actor Object Reference, Mesh:Dynamic Mesh Object Reference=@n2) -> (then, ReturnValue:bool)
   > n3 = GetDynamicMesh(self=Get DynamicMeshComponent1() -> (DynamicMeshComponent1:Dynamic Mesh Component Object Reference))
   > ReleaseComputeMesh(Mesh=@n3)
   > n4 = GetDynamicMesh(self=Get DynamicMeshComponent2() -> (DynamicMeshComponent2:Dynamic Mesh Component Object Reference))
   > ReleaseComputeMesh(Mesh=@n4)
   n5 = GetGlobalMeshPool!(execute=@n1.then_1) -> (then, ReturnValue:Dynamic Mesh Pool Object Reference)
   > n6 = RequestMesh!(execute, self:Dynamic Mesh Pool Object Reference=@n5) -> (then, ReturnValue:Dynamic Mesh Object Reference)
   > n7 = Copy Mesh From Static Mesh!(
    execute
    FromStaticMeshAsset:Static Mesh Object Reference=&/SM_BugProto_TA2
    ToDynamicMesh:Dynamic Mesh Object Reference=@n6
    AssetOptions:Geometry Script Copy Mesh From Asset Options Structure
    RequestedLOD:Geometry Script Mesh Read LOD Structure=Make Geometry Script Mesh Read LOD!(LODType:EGeometryScriptLODType Enum=SourceModel, LODIndex:int=0) -> (GeometryScriptMeshReadLOD:Geometry Script Mesh Read LOD Structure) {is=title K2Node_MakeStruct}
    bUseSectionMaterials:bool=true
   ) -> (Failure:Exec, Success:Exec, ReturnValue:Dynamic Mesh Object Reference)
   n8 = CopyMeshToStaticMesh!(
    execute=@n7.Success
    FromDynamicMesh:Dynamic Mesh Object Reference=@n7
    ToStaticMeshAsset:Static Mesh Object Reference=&/SM_BugProto_Generated
    Options:Geometry Script Copy Mesh To Asset Options Structure
    TargetLOD:Geometry Script Mesh Write LOD Structure=Make Geometry Script Mesh Write LOD!(bWriteHiResSource:bool=False, LODIndex:int=0) -> (GeometryScriptMeshWriteLOD:Geometry Script Mesh Write LOD Structure) {is=title K2Node_MakeStruct}
    bUseSectionMaterials:bool=true
   ) -> (Failure:Exec, Success:Exec, ReturnValue:Dynamic Mesh Object Reference)
   SetDynamicMesh!(execute=@n8.Success, self:Dynamic Mesh Component Object Reference=Get DynamicMeshComponent(), NewMesh:Dynamic Mesh Object Reference=@n8) -> (then)
   n9 = GetGlobalMeshPool(execute=@n1.then_2)
   > n10 = RequestMesh(self=@n9)
   > n11 = Copy Mesh From Static Mesh(
    FromStaticMeshAsset=&SM_BugProto_TA2
    ToDynamicMesh=@n10
    RequestedLOD=Make Geometry Script Mesh Read LOD(LODType=SourceModel, LODIndex=0)
    bUseSectionMaterials=true
   )
   n12 = ApplySimplifyToVertexCount!(
    execute=@n11.Success
    TargetMesh:Dynamic Mesh Object Reference=@n11
    VertexCount:int=400
    Options:Geometry Script Simplify Mesh Options Structure=Make Geometry Script Simplify Mesh Options!(
     Method:EGeometryScriptRemoveMeshSimplificationType Enum=Get Method1() -> (Method1:EGeometryScriptRemoveMeshSimplificationType Enum=StandardQEM)
     bAllowSeamCollapse:bool=True
     bAllowSeamSmoothing:bool=True
     bAllowSeamSplits:bool=True
     bPreserveVertexPositions:bool=False
     bRetainQuadricMemory:bool=False
     EdgeLengthWeightMap:Geometry Script Weight Map Density Structure=(Handle=(WeightMapAttributeLayerIndex=-1),RelativeDensity=0)
     GeometricToleranceWeightMap:Geometry Script Weight Map Density Structure=(Handle=(WeightMapAttributeLayerIndex=-1),RelativeDensity=0)
     QuadricErrorWeightMap:Geometry Script Weight Map Density Structure=(Handle=(WeightMapAttributeLayerIndex=-1),RelativeDensity=0)
     RegularizeWeight:float=0
     bAutoCompact:bool=True
     QuadricVariant:EGeometryScriptMeshSimplificationQuadricVariant Enum=PlaneQuadric
     NormalAttributeWeight:float=16
     TangentAttributeWeight:float=0.1
     ColorAttributeWeight:float=0.1
     TexCoordAttributeWeight:float=0.5
     ScaleCorrection:float=1
    ) -> (GeometryScriptSimplifyMeshOptions:Geometry Script Simplify Mesh Options Structure) {is=title K2Node_MakeStruct}
   ) -> (then, ReturnValue:Dynamic Mesh Object Reference)
   > n13 = CopyMeshToStaticMesh(
    FromDynamicMesh=@n12
    ToStaticMeshAsset=&SM_BugProto_Generated
    TargetLOD=Make Geometry Script Mesh Write LOD(bWriteHiResSource=False, LODIndex=1)
    bUseSectionMaterials=true
   )
   SetDynamicMesh(execute=@n13.Success, self=Get DynamicMeshComponent1(), NewMesh=@n13)
   n14 = GetGlobalMeshPool(execute=@n1.then_3)
   > n15 = RequestMesh(self=@n14)
   > n16 = Copy Mesh From Static Mesh(
    FromStaticMeshAsset=&SM_BugProto_TA2
    ToDynamicMesh=@n15
    RequestedLOD=Make Geometry Script Mesh Read LOD(LODType=SourceModel, LODIndex=0)
    bUseSectionMaterials=true
   )
   n17 = ApplySimplifyToVertexCount(
    execute=@n16.Success
    TargetMesh=@n16
    VertexCount=100
    Options=Make Geometry Script Simplify Mesh Options(
     Method=Get Method2() -> (Method2:EGeometryScriptRemoveMeshSimplificationType Enum=StandardQEM)
     bAllowSeamCollapse=True
     bAllowSeamSmoothing=True
     bAllowSeamSplits=True
     bPreserveVertexPositions=False
     bRetainQuadricMemory=False
     EdgeLengthWeightMap=(Handle=(WeightMapAttributeLayerIndex=-1),RelativeDensity=0)
     GeometricToleranceWeightMap=(Handle=(WeightMapAttributeLayerIndex=-1),RelativeDensity=0)
     QuadricErrorWeightMap=(Handle=(WeightMapAttributeLayerIndex=-1),RelativeDensity=0)
     RegularizeWeight=0
     bAutoCompact=True
     QuadricVariant=PlaneQuadric
     NormalAttributeWeight=16
     TangentAttributeWeight=0.1
     ColorAttributeWeight=0.1
     TexCoordAttributeWeight=0.5
     ScaleCorrection=1
    )
   )
   > n18 = CopyMeshToStaticMesh(
    FromDynamicMesh=@n17
    ToStaticMeshAsset=&SM_BugProto_Generated
    TargetLOD=Make Geometry Script Mesh Write LOD(bWriteHiResSource=False, LODIndex=2)
    bUseSectionMaterials=true
   )
   SetDynamicMesh(execute=@n18.Success, self=Get DynamicMeshComponent2(), NewMesh=@n18)
  ]}
 UserConstructionScript {nodes [Construction Script()]}
]
variables [
 Method1 {type=EGeometryScriptRemoveMeshSimplificationType, default=VolumePreserving}
 Method2 {type=EGeometryScriptRemoveMeshSimplificationType, default=AttributeAware}
]
